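# Control Rig deform rig for the SesameSeed medium male player body, correcting body and cape bones.
# The live chain runs a shared DeformRig function, then parent-constrains five dynamic cape master bones to weighted spine, shoulder and arm bones.
# Spherical pose readers on left clavicle and both upper arms weight cape offsets, followed by an IsSkydiving-driven offset.
# Aim constraints and older-style cape corrections for backpack, parachute and sprint states sit unconnected to execution.
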ControlRigBlueprint /Game/Characters/Player/Male/Medium/Bodies/M_MED_SesameSeed/Meshes/M_MED_SesameSeed_DeformRig
graphs [
 RigVMFunctionLibrary {links [], nodes []}
 RigVMModel {nodes [
   &RigUnit_AimConstraintLocalSpaceOffset(
    Child=(Type=Bone,Name="dyn_cape_mid_l_MSTR")
    ExecutePin=()
    Filter=(bX=True,bY=True,bZ=True)
    Parents=((Item=(Type=Bone,Name="pelvis"),Weight=1))
    Weight=1
    WorldUp=(Target=(X=0,Y=0,Z=1),Kind=Location,Space=(Type=Bone,Name="pelvis"))
   )
   > &RigUnit_AimConstraintLocalSpaceOffset(
    Child=(Type=Bone,Name="dyn_cape_l_MSTR")
    Filter=(bX=True,bY=True,bZ=True)
    Parents=((Item=(Type=Bone,Name="pelvis"),Weight=1))
    Weight=1
    WorldUp=(Target=(X=0,Y=0,Z=1),Kind=Location,Space=(Type=Bone,Name="pelvis"))
   )
   > &RigUnit_AimConstraintLocalSpaceOffset(
    Child=(Type=Bone,Name="dyn_cape_mid_r_MSTR")
    Filter=(bX=True,bY=True,bZ=True)
    Parents=((Item=(Type=Bone,Name="pelvis"),Weight=1))
    Weight=1
    WorldUp=(Target=(X=0,Y=0,Z=1),Kind=Location,Space=(Type=Bone,Name="pelvis"))
   )
   > &RigUnit_AimConstraintLocalSpaceOffset(
    Child=(Type=Bone,Name="dyn_cape_r_MSTR")
    Filter=(bX=True,bY=True,bZ=True)
    Parents=((Item=(Type=Bone,Name="pelvis"),Weight=1))
    Weight=1
    WorldUp=(Target=(X=0,Y=0,Z=1),Kind=Location,Space=(Type=Bone,Name="pelvis"))
   )
   > &RigUnit_AimConstraintLocalSpaceOffset(
    Child=(Type=Bone,Name="dyn_cape_MSTR")
    Filter=(bX=True,bY=True,bZ=True)
    Parents=((Item=(Type=Bone,Name="pelvis"),Weight=1))
    Weight=1
    WorldUp=(Target=(X=0,Y=0,Z=1),Kind=Location,Space=(Type=Bone,Name="pelvis"))
   )
   RigUnit_BeginExecution = &RigUnit_BeginExecution()
   FN_DeformRig_FNC = RigVMFunctionReferenceNode(
    DeformRig_ Curve="(deformRig_l_upperarm_up_alpha_72_#1=False,deformRig_r_upperarm_up_alpha_73_#2=False,deformRig_l_upperarm_fr_alpha_74_#3=False,deformRig_r_upperarm_fr_alpha_75_CB51B8 ...(truncated)"
    DeformRig_Bool="(upperarm_1_#4=True,elbow_3_#5=True,wrist_7_#6=True,groin_8_#7=True,glute_17_#8=True,knee_9_40C408324CA261 ...(truncated)"
    DeformRig_Debug=(upperarm_debug_18_#4=False,wrist_debug_20_#6=False,lat_debug_26_#9=False,groinGlute_debug_27_#7=False)
    DeformRig_Value="(upperarm_up_18_#10=(Rotation=(X=0,Y=0,Z=0,W=1),Translation=(X=0,Y=0,Z=5),Scale3D=(X=0.7,Y=1,Z=1)),upperarm_fwd_fr_19_#11=(Rotation=(X=0,Y=0,Z=0,W=1),Translation=(X=-10,Y=10,Z=0),Scale3D=(X=0.5,Y=1, ...(truncated)"
    ExecuteContext=@RigUnit_BeginExecution.ExecutePin
   )
   > &RigUnit_ParentConstraint(
    AdvancedSettings=(RotationOrderForFilter=XZY)
    Child=(Type=Bone,Name="dyn_cape_MSTR")
    Filter=(TranslationFilter=(bX=True,bY=True,bZ=True),RotationFilter=(bX=false,bY=false,bZ=false),ScaleFilter=(bX=True,bY=True,bZ=True))
    Parents=((Item=(Type=Bone,Name="clavicle_r"),Weight=0.004),(Item=(Type=Bone,Name="spine_02"),Weight=0.01),(Item=(Type=Bone,Name="spine_04"),Weight=0.156),(Item=(Type=Bone,Name="spine_03"),Weight=0.4144),(Item=(Type=Bone,Name="spine_05"),Weight=0.4156))
    Weight=1
   )
   > &RigUnit_ParentConstraint(
    AdvancedSettings=(RotationOrderForFilter=XZY)
    Child=(Type=Bone,Name="dyn_cape_mid_l_MSTR")
    Filter=(TranslationFilter=(bX=True,bY=True,bZ=True),RotationFilter=(bX=false,bY=false,bZ=false),ScaleFilter=(bX=True,bY=True,bZ=True))
    Parents="((Item=(Type=Bone,Name=\"deform_upperarm_up_l\"),Weight=0.0201),(Item=(Type=Bone,Name=\"deform_upperarm_bk_l\"),Weight=0.0446),(Item=(Type=Bone,Name=\"upperarm_twist_01_l\"),Weight=0.0541),(Item=(Type=Bone,Name=\"spine_04\"),Weight=0.0832),(Item=(Type=Bone,Name=\"s ...(truncated)"
    Weight=1
   )
   > &RigUnit_ParentConstraint(
    AdvancedSettings=(RotationOrderForFilter=XZY)
    Child=(Type=Bone,Name="dyn_cape_l_MSTR")
    Filter=(TranslationFilter=(bX=True,bY=True,bZ=True),RotationFilter=(bX=false,bY=false,bZ=false),ScaleFilter=(bX=True,bY=True,bZ=True))
    Parents="((Item=(Type=Bone,Name=\"spine_05\"),Weight=0.016),(Item=(Type=Bone,Name=\"spine_04\"),Weight=0.0219),(Item=(Type=Bone,Name=\"upperarm_twist_02_l\"),Weight=0.0438),(Item=(Type=Bone,Name=\"spine_03\"),Weight=0.0584),(Item=(Type=Bone,Name=\"deform_upperarm_up_l\"),Wei ...(truncated)"
    Weight=1
   )
   > &RigUnit_ParentConstraint(
    AdvancedSettings=(RotationOrderForFilter=XZY)
    Child=(Type=Bone,Name="dyn_cape_mid_r_MSTR")
    Filter=(TranslationFilter=(bX=True,bY=True,bZ=True),RotationFilter=(bX=false,bY=false,bZ=false),ScaleFilter=(bX=True,bY=True,bZ=True))
    Parents="((Item=(Type=Bone,Name=\"deform_upperarm_up_r\"),Weight=0.0058),(Item=(Type=Bone,Name=\"spine_02\"),Weight=0.0113),(Item=(Type=Bone,Name=\"upperarm_twist_01_r\"),Weight=0.0442),(Item=(Type=Bone,Name=\"deform_upperarm_bk_r\"),Weight=0.0533),(Item=(Type=Bone,Name=\"s ...(truncated)"
    Weight=1
   )
   > &RigUnit_ParentConstraint(
    AdvancedSettings=(RotationOrderForFilter=XZY)
    Child=(Type=Bone,Name="dyn_cape_r_MSTR")
    Filter=(TranslationFilter=(bX=True,bY=True,bZ=True),RotationFilter=(bX=false,bY=false,bZ=false),ScaleFilter=(bX=True,bY=True,bZ=True))
    Parents="((Item=(Type=Bone,Name=\"spine_05\"),Weight=0.0152),(Item=(Type=Bone,Name=\"spine_04\"),Weight=0.0176),(Item=(Type=Bone,Name=\"upperarm_twist_02_r\"),Weight=0.0377),(Item=(Type=Bone,Name=\"spine_03\"),Weight=0.0464),(Item=(Type=Bone,Name=\"deform_upperarm_up_r\"),We ...(truncated)"
    Weight=1
   )
   > n1 = &RigUnit_SphericalPoseReader(
    ActiveRegionScaleFactors=(PositiveWidth=3,NegativeWidth=1,PositiveHeight=1.5,NegativeHeight=1.5)
    ActiveRegionSize=0.1
    Debug=(bDrawDebug=true,DebugScale=25,DebugSegments=20,DebugThickness=0.25)
    DriverAxis=(X=1,Y=0,Z=0)
    DriverItem=(Type=Bone,Name="upperarm_l")
    FalloffRegionScaleFactors=(PositiveWidth=1,NegativeWidth=1,PositiveHeight=1,NegativeHeight=1)
    FalloffSize=0.35
    OptionalParentItem=(Type=Bone,Name="clavicle_l")
    RotationOffset=(X=120,Y=20,Z=220)
   )
   > &RigUnit_ModifyTransforms(
    ItemToModify="((Item=(Type=Bone,Name=\"dyn_cape_l_MSTR\"),Transform=(Rotation=(X=0.3551,Y=0,Z=0,W=0.9348),Translation=(X=0,Y=-2.2187,Z=2.4992),Scale3D=(X=1,Y=1,Z=1))),(Item=(Type=Bone,Name=\"dyn_cape_mid_l_MSTR\"),Transform=(Rotation=(X=-0.125,Y=-0.0378,Z=-0.0048,W=-0.9914) ...(truncated)"
    WeightMinimum=0
    Weight=@n1.OutputParam
   )
   > n2 = &RigUnit_SphericalPoseReader(
    ActiveRegionScaleFactors=(PositiveWidth=1,NegativeWidth=1,PositiveHeight=1,NegativeHeight=1)
    ActiveRegionSize=0.1
    Debug=(bDrawDebug=true,bDraw2D=False,bDrawLocalAxes=False,DebugScale=25,DebugSegments=20,DebugThickness=0.25)
    DriverAxis=(X=1,Y=0,Z=0)
    DriverItem=(Type=Bone,Name="clavicle_l")
    FalloffRegionScaleFactors=(PositiveWidth=1,NegativeWidth=1,PositiveHeight=1,NegativeHeight=1)
    FalloffSize=0.2
    FlipHeightScaling=False
    FlipWidthScaling=False
    OptionalParentItem=(Type=Bone,Name="spine_05")
    RotationOffset=(X=-90,Y=0,Z=50)
   )
   > &RigUnit_ModifyTransforms(
    ItemToModify="((Item=(Type=Bone,Name=\"dyn_cape_l_MSTR\"),Transform=(Rotation=(X=0,Y=0,Z=0,W=1),Translation=(X=0,Y=0,Z=0),Scale3D=(X=1,Y=1,Z=1))),(Item=(Type=Bone,Name=\"dyn_cape_mid_l_MSTR\"),Transform=(Rotation=(X=0,Y=0,Z=0,W=1),Translation=(X=0,Y=0,Z=0),Scale3D=(X=1,Y=1, ...(truncated)"
    WeightMinimum=0
    Weight=@n2.OutputParam
   )
   > n3 = &RigUnit_SphericalPoseReader(
    ActiveRegionScaleFactors=(PositiveWidth=3,NegativeWidth=1,PositiveHeight=1.5,NegativeHeight=1.5)
    ActiveRegionSize=0.1
    Debug=(bDrawDebug=true,DebugScale=25,DebugSegments=20,DebugThickness=0.25)
    DriverAxis=(X=-1,Y=0,Z=0)
    DriverItem=(Type=Bone,Name="upperarm_r")
    FalloffRegionScaleFactors=(PositiveWidth=1,NegativeWidth=1,PositiveHeight=1,NegativeHeight=1)
    FalloffSize=0.35
    FlipHeightScaling=false
    FlipWidthScaling=true
    OptionalParentItem=(Type=Bone,Name="clavicle_r")
    RotationOffset=(X=-60,Y=20,Z=-140)
   )
   > &RigUnit_ModifyTransforms(
    ItemToModify="((Item=(Type=Bone,Name=\"dyn_cape_r_MSTR\"),Transform=(Rotation=(X=0.3551,Y=0,Z=0,W=0.9348),Translation=(X=0,Y=2.2187,Z=-2.4992),Scale3D=(X=1,Y=1,Z=1))),(Item=(Type=Bone,Name=\"dyn_cape_mid_r_MSTR\"),Transform=(Rotation=(X=-0.125,Y=-0.0378,Z=-0.0048,W=-0.9914) ...(truncated)"
    WeightMinimum=0
    Weight=@n3.OutputParam
   )
   RigVMFunction_MathBoolToFloat = &RigVMFunction_MathBoolToFloat(Value=RigVMVariableNode() {variable=IsSkydiving}.Value)
   > &RigUnit_ModifyTransforms(
    ItemToModify="((Item=(Type=Bone,Name=\"dyn_cape_MSTR\"),Transform=(Rotation=(X=0,Y=0,Z=0,W=1),Translation=(X=0,Y=0,Z=10),Scale3D=(X=1,Y=1,Z=1))),(Item=(Type=Bone,Name=\"dyn_cape_mid_r_MSTR\"),Transform=(Rotation=(X=0,Y=0,Z=0.1305,W=0.9914),Translation=(X=0,Y=0,Z=0),Scale3D= ...(truncated)"
    WeightMinimum=0
    Weight=@RigVMFunction_MathBoolToFloat.Result
   )
   n4 = &RigVMFunction_ControlFlowBranch(ExecuteContext=(), Condition=RigVMVariableNode() {variable=backpack_wearing}.Value)
   n5 = &RigUnit_SphericalPoseReader(
    ActiveRegionScaleFactors=(PositiveWidth=0.5,NegativeWidth=0.5,PositiveHeight=0.5,NegativeHeight=0.5)
    ActiveRegionSize=0.1
    Debug=(bDrawDebug=false,bDraw2D=false,bDrawLocalAxes=false,DebugScale=25,DebugSegments=20,DebugThickness=0.25)
    DriverAxis=(X=1,Y=0,Z=0)
    DriverItem=(Type=Bone,Name="upperarm_l")
    FalloffRegionScaleFactors=(PositiveWidth=0.2,NegativeWidth=0.2,PositiveHeight=0.2,NegativeHeight=0.2)
    FalloffSize=1
    FlipWidthScaling=False
    OptionalParentItem=(Type=Bone)
    RotationOffset=(X=-20,Y=-120,Z=0)
    ExecutePin=@n4.False
   )
   > &RigUnit_ModifyTransforms(
    ItemToModify=((Item=(Type=Bone,Name="cape_02_l"),Transform=(Rotation=(X=0.067,Y=0.25,Z=-0.25,W=0.933),Translation=(X=0,Y=-5,Z=0),Scale3D=(X=1,Y=1,Z=1))))
    Weight=@n5.OutputParam
   )
   > n6 = &RigUnit_ModifyTransforms(
    ItemToModify=((Item=(Type=Bone,Name="cape_mid_01_l"),Transform=(Rotation=(X=0.067,Y=0.25,Z=-0.25,W=0.933),Translation=(X=0,Y=-2,Z=0),Scale3D=(X=1,Y=1,Z=1))))
    Weight=@n5.OutputParam
   )
   RigVMFunction_Sequence = &RigVMFunction_Sequence(ExecuteContext=@n6.ExecutePin)
   &RigUnit_ModifyTransforms(
    ItemToModify=((Item=(Type=Bone,Name="calf_r_extra"),Transform=(Rotation=(X=0,Y=0,Z=0,W=1),Translation=(X=0,Y=0,Z=0),Scale3D=(X=1,Y=1,Z=0.5))))
    Weight=1
    ExecutePin=@RigVMFunction_Sequence.A
   )
   > n7 = &RigUnit_SphericalPoseReader(
    ActiveRegionScaleFactors=(PositiveWidth=0.5,NegativeWidth=0.5,PositiveHeight=0.5,NegativeHeight=0.5)
    ActiveRegionSize=0.1
    Debug=(bDrawDebug=false,bDraw2D=false,bDrawLocalAxes=false,DebugScale=25,DebugSegments=20,DebugThickness=0.25)
    DriverAxis=(X=-1,Y=0,Z=0)
    DriverItem=(Type=Bone,Name="spine_05")
    FalloffRegionScaleFactors=(PositiveWidth=0.2,NegativeWidth=0.2,PositiveHeight=0.2,NegativeHeight=0.2)
    FalloffSize=1
    FlipWidthScaling=False
    OptionalParentItem=(Type=Bone)
    RotationOffset=(X=20,Y=90,Z=0)
   )
   n8 = &RigVMFunction_AlphaInterp(InterpSpeedDecreasing=10, InterpSpeedIncreasing=3, Scale=1, bClampResult=False, Value=RigVMVariableNode() {variable=parachute}.Value)
   Multiply = &RigVMFunction_MathDoubleMul(B=@n8.Result, A=@n7.OutputParam)
   > &RigUnit_ModifyTransforms(
    ItemToModify=((Item=(Type=Bone,Name="cape_mid_01_l"),Transform=(Rotation=(X=0,Y=0,Z=-0.2588,W=0.9659),Translation=(X=0,Y=0,Z=0),Scale3D=(X=1,Y=1,Z=1))))
    Weight=@Multiply.Result
   )
   > &RigUnit_ModifyTransforms(
    ItemToModify=((Item=(Type=Bone,Name="cape_mid_01_r"),Transform=(Rotation=(X=0,Y=0,Z=-0.2588,W=0.9659),Translation=(X=0,Y=0,Z=0),Scale3D=(X=1,Y=1,Z=0.5))))
    Weight=@Multiply.Result
   )
   n9 = &RigVMFunction_AlphaInterp(
    InterpSpeedDecreasing=5
    InterpSpeedIncreasing=5
    Scale=1
    bClampResult=False
    Value=RigVMVariableNode() {variable=tacticalSprint_and_jump}.Value
   )
   > &RigUnit_ModifyTransforms(
    ItemToModify=((Item=(Type=Bone,Name="cape_mid_01_l"),Transform=(Rotation=(X=0,Y=0,Z=-0.2588,W=0.9659),Translation=(X=0,Y=0,Z=0),Scale3D=(X=1,Y=1,Z=1))))
    Weight=@n9.Result
   )
   > &RigUnit_ModifyTransforms(
    ItemToModify=((Item=(Type=Bone,Name="cape_mid_01_r"),Transform=(Rotation=(X=0,Y=0,Z=-0.2588,W=0.9659),Translation=(X=0,Y=0,Z=0),Scale3D=(X=1,Y=1,Z=1))))
    Weight=@n9.Result
   )
   > n10 = &RigUnit_SphericalPoseReader(
    ActiveRegionScaleFactors=(PositiveWidth=1,NegativeWidth=1,PositiveHeight=1,NegativeHeight=1)
    ActiveRegionSize=0.04
    Debug=(bDrawDebug=false,bDraw2D=false,bDrawLocalAxes=false,DebugScale=25,DebugSegments=20,DebugThickness=0.25)
    DriverAxis=(X=-1,Y=0,Z=0)
    DriverItem=(Type=Bone,Name="spine_05")
    FalloffRegionScaleFactors=(PositiveWidth=0.3,NegativeWidth=0.3,PositiveHeight=0.3,NegativeHeight=0.3)
    FalloffSize=1
    FlipWidthScaling=False
    OptionalParentItem=(Type=Bone)
    RotationOffset=(X=20,Y=90,Z=-40)
   )
   n11 = &RigVMFunction_AlphaInterp(InterpSpeedDecreasing=10, InterpSpeedIncreasing=3, Scale=1, bClampResult=False, Value=@n10.OutputParam)
   > &RigUnit_ModifyTransforms(
    ItemToModify=((Item=(Type=Bone,Name="cape_mid_01_l"),Transform=(Rotation=(X=0,Y=0,Z=-0.2588,W=0.9659),Translation=(X=0,Y=-10,Z=0),Scale3D=(X=1,Y=1,Z=1))))
    Weight=@n11.Result
   )
   > &RigUnit_ModifyTransforms(
    ItemToModify=((Item=(Type=Bone,Name="cape_mid_01_r"),Transform=(Rotation=(X=0,Y=0,Z=-0.2588,W=0.9659),Translation=(X=0,Y=-10,Z=0),Scale3D=(X=1,Y=1,Z=0.5))))
    Weight=@n11.Result
   )
   > n12 = &RigUnit_SphericalPoseReader(
    ActiveRegionScaleFactors=(PositiveWidth=1,NegativeWidth=1,PositiveHeight=1,NegativeHeight=1)
    ActiveRegionSize=0.1
    Debug=(bDrawDebug=false,bDraw2D=false,bDrawLocalAxes=false,DebugScale=25,DebugSegments=20,DebugThickness=0.25)
    DriverAxis=(X=0,Y=1,Z=0)
    DriverItem=(Type=Bone,Name="clavicle_l")
    FalloffRegionScaleFactors=(PositiveWidth=0.3,NegativeWidth=0.3,PositiveHeight=0.3,NegativeHeight=0.3)
    FalloffSize=0.3
    FlipWidthScaling=False
    OptionalParentItem=(Type=None)
    RotationOffset=(X=47.8405,Y=41.5052,Z=21.3912)
   )
   > &RigUnit_ModifyTransforms(
    ItemToModify=((Item=(Type=Bone,Name="cape_01_l"),Transform=(Rotation=(X=-0.0795,Y=0.0728,Z=-0.3164,W=0.9425),Translation=(X=0,Y=0,Z=0),Scale3D=(X=1,Y=1,Z=1))))
    Weight=@n12.OutputParam
   )
   n13 = &RigUnit_SphericalPoseReader(
    ActiveRegionScaleFactors=(PositiveWidth=0.5,NegativeWidth=0.5,PositiveHeight=0.5,NegativeHeight=0.5)
    ActiveRegionSize=0.1
    Debug=(bDrawDebug=false,bDraw2D=false,bDrawLocalAxes=false,DebugScale=25,DebugSegments=20,DebugThickness=0.25)
    DriverAxis=(X=-1,Y=0,Z=0)
    DriverItem=(Type=Bone,Name="upperarm_r")
    FalloffRegionScaleFactors=(PositiveWidth=0.2,NegativeWidth=0.2,PositiveHeight=0.2,NegativeHeight=0.1)
    FalloffSize=1
    FlipWidthScaling=False
    OptionalParentItem=(Type=Bone)
    RotationOffset=(X=20,Y=73,Z=0)
    ExecutePin=@RigVMFunction_Sequence.B
   )
   n14 = &RigVMFunction_ControlFlowBranch(ExecuteContext=@n13.ExecutePin, Condition=RigVMVariableNode() {variable=TacticalSprinting}.Value)
   n15 = &RigVMFunction_AlphaInterp(InterpSpeedDecreasing=8, InterpSpeedIncreasing=8, Scale=1, bClampResult=False, Value=@n13.OutputParam)
   &RigUnit_ModifyTransforms(
    ItemToModify=((Item=(Type=Bone,Name="cape_02_r"),Transform=(Rotation=(X=-0.0302,Y=-0.171,Z=-0.171,W=0.9698),Translation=(X=0,Y=-5,Z=0),Scale3D=(X=1,Y=1,Z=1))))
    Weight=@n15.Result
    ExecutePin=@n14.False
   )
   n16 = &RigVMFunction_AlphaInterp(InterpSpeedDecreasing=8, InterpSpeedIncreasing=8, Scale=1, bClampResult=False, Value=@n13.OutputParam)
   > &RigUnit_ModifyTransforms(
    ItemToModify=((Item=(Type=Bone,Name="cape_mid_01_r"),Transform=(Rotation=(X=-0.0302,Y=-0.171,Z=-0.171,W=0.9698),Translation=(X=0,Y=-2,Z=0),Scale3D=(X=1,Y=1,Z=1))))
    Weight=@n16.Result
   )
   &RigUnit_ModifyTransforms(
    ItemToModify=((Item=(Type=Bone,Name="cape_02_r"),Transform=(Rotation=(X=-0.0302,Y=-0.171,Z=-0.171,W=0.9698),Translation=(X=0,Y=-5,Z=0),Scale3D=(X=1,Y=1,Z=1))))
    ExecutePin=@n14.True
    Weight=@n13.OutputParam
   )
   > &RigUnit_ModifyTransforms(
    ItemToModify=((Item=(Type=Bone,Name="cape_mid_01_r"),Transform=(Rotation=(X=-0.0302,Y=-0.171,Z=-0.171,W=0.9698),Translation=(X=0,Y=-2,Z=0),Scale3D=(X=1,Y=1,Z=1))))
    Weight=@n13.OutputParam
   )
   n17 = &RigUnit_SphericalPoseReader(
    ActiveRegionScaleFactors=(PositiveWidth=1,NegativeWidth=1,PositiveHeight=1,NegativeHeight=1)
    ActiveRegionSize=0.08
    Debug=(bDrawDebug=false,bDraw2D=false,bDrawLocalAxes=false,DebugScale=25,DebugSegments=20,DebugThickness=0.25)
    DriverAxis=(X=-1,Y=0,Z=0)
    DriverItem=(Type=Bone,Name="spine_05")
    FalloffRegionScaleFactors=(PositiveWidth=0.3,NegativeWidth=0.3,PositiveHeight=0.3,NegativeHeight=0.3)
    FalloffSize=1
    FlipWidthScaling=False
    OptionalParentItem=(Type=Bone,Name="None")
    RotationOffset=(X=20,Y=90,Z=-40)
    ExecutePin=@n4.True
   )
   n18 = &RigVMFunction_AlphaInterp(InterpSpeedDecreasing=10, InterpSpeedIncreasing=3, Scale=1, bClampResult=False, Value=@n17.OutputParam)
   > &RigUnit_ModifyTransforms(
    ItemToModify=((Item=(Type=Bone,Name="cape_mid_01_l"),Transform=(Rotation=(X=0,Y=0,Z=-0.2588,W=0.9659),Translation=(X=0,Y=-10,Z=0),Scale3D=(X=1,Y=1,Z=1))))
    Weight=@n18.Result
   )
   > &RigUnit_ModifyTransforms(
    ItemToModify=((Item=(Type=Bone,Name="cape_mid_01_r"),Transform=(Rotation=(X=0,Y=0,Z=-0.2588,W=0.9659),Translation=(X=0,Y=-10,Z=0),Scale3D=(X=1,Y=1,Z=0.5))))
    Weight=@n18.Result
   )
   > &RigUnit_ModifyTransforms(
    ItemToModify=((Item=(Type=Bone,Name="cape_02_l"),Transform=(Rotation=(X=0,Y=0,Z=-0.2588,W=0.9659),Translation=(X=0,Y=-10,Z=0),Scale3D=(X=1,Y=1,Z=1))))
    Weight=@n18.Result
   )
   > &RigUnit_ModifyTransforms(
    ItemToModify=((Item=(Type=Bone,Name="cape_02_r"),Transform=(Rotation=(X=0,Y=0,Z=-0.2588,W=0.9659),Translation=(X=0,Y=-10,Z=0),Scale3D=(X=1,Y=1,Z=0.5))))
    Weight=@n18.Result
   )
   > n19 = &RigUnit_SphericalPoseReader(
    ActiveRegionScaleFactors=(PositiveWidth=0.5,NegativeWidth=0.5,PositiveHeight=0.5,NegativeHeight=0.5)
    ActiveRegionSize=0.1
    Debug=(bDrawDebug=false,bDraw2D=false,bDrawLocalAxes=false,DebugScale=25,DebugSegments=20,DebugThickness=0.25)
    DriverAxis=(X=1,Y=0,Z=0)
    DriverItem=(Type=Bone,Name="upperarm_l")
    FalloffRegionScaleFactors=(PositiveWidth=0.2,NegativeWidth=0.2,PositiveHeight=0.2,NegativeHeight=0.2)
    FalloffSize=1
    FlipWidthScaling=False
    OptionalParentItem=(Type=Bone)
    RotationOffset=(X=-20,Y=-120,Z=0)
   )
   > &RigUnit_ModifyTransforms(
    ItemToModify=((Item=(Type=Bone,Name="cape_02_l"),Transform=(Rotation=(X=0,Y=0,Z=-0.0872,W=0.9962),Translation=(X=0,Y=0,Z=-7),Scale3D=(X=1,Y=1,Z=1))))
    Weight=@n19.OutputParam
   )
   > &RigUnit_ModifyTransforms(
    ItemToModify=((Item=(Type=Bone,Name="cape_mid_01_l"),Transform=(Rotation=(X=0,Y=0,Z=-0.0872,W=0.9962),Translation=(X=0,Y=-2,Z=0),Scale3D=(X=1,Y=1,Z=1))))
    Weight=@n19.OutputParam
   )
   > n20 = &RigUnit_SphericalPoseReader(
    ActiveRegionScaleFactors=(PositiveWidth=0.5,NegativeWidth=0.5,PositiveHeight=0.5,NegativeHeight=0.5)
    ActiveRegionSize=0.1
    Debug=(bDrawDebug=false,bDraw2D=false,bDrawLocalAxes=false,DebugScale=25,DebugSegments=20,DebugThickness=0.25)
    DriverAxis=(X=-1,Y=0,Z=0)
    DriverItem=(Type=Bone,Name="upperarm_r")
    FalloffRegionScaleFactors=(PositiveWidth=0.2,NegativeWidth=0.2,PositiveHeight=0.2,NegativeHeight=0.2)
    FalloffSize=1
    FlipWidthScaling=False
    OptionalParentItem=(Type=Bone)
    RotationOffset=(X=20,Y=60,Z=0)
   )
   > &RigUnit_ModifyTransforms(
    ItemToModify=((Item=(Type=Bone,Name="cape_02_r"),Transform=(Rotation=(X=0,Y=0,Z=-0.0872,W=0.9962),Translation=(X=0,Y=0,Z=7),Scale3D=(X=1,Y=1,Z=1))))
    Weight=@n20.OutputParam
   )
   > &RigUnit_ModifyTransforms(
    ItemToModify=((Item=(Type=Bone,Name="cape_mid_01_r"),Transform=(Rotation=(X=0,Y=0,Z=-0.0872,W=0.9962),Translation=(X=0,Y=-2,Z=0),Scale3D=(X=1,Y=1,Z=1))))
    Weight=@n20.OutputParam
   )
   n21 = &RigUnit_SphericalPoseReader(
    ActiveRegionScaleFactors=(PositiveWidth=1,NegativeWidth=1,PositiveHeight=1,NegativeHeight=1)
    ActiveRegionSize=0.1
    Debug=(bDrawDebug=false,bDraw2D=False,bDrawLocalAxes=False,DebugScale=25,DebugSegments=20,DebugThickness=0.25)
    DriverAxis=(X=1,Y=0,Z=0)
    DriverItem=(Type=Bone,Name="clavicle_r")
    ExecutePin=()
    FalloffRegionScaleFactors=(PositiveWidth=1,NegativeWidth=1,PositiveHeight=1,NegativeHeight=1)
    FalloffSize=0.2
    FlipHeightScaling=False
    FlipWidthScaling=False
    OptionalParentItem=(Type=Bone,Name="spine_05")
    RotationOffset=(X=90,Y=0,Z=50)
   )
   > &RigUnit_ModifyTransforms(
    ItemToModify=((Item=(Type=Bone,Name="def_cloth_clavicle_r"),Transform=(Rotation=(X=0,Y=0,Z=0,W=1),Translation=(X=0,Y=0,Z=0),Scale3D=(X=1,Y=1,Z=1))))
    WeightMinimum=0
    Weight=@n21.OutputParam
   )
   n22 = &RigUnit_SphericalPoseReader(
    ActiveRegionScaleFactors=(PositiveWidth=3,NegativeWidth=1,PositiveHeight=1.5,NegativeHeight=1.5)
    ActiveRegionSize=0.1
    Debug=(bDrawDebug=false,DebugScale=25,DebugSegments=20,DebugThickness=0.25)
    DriverAxis=(X=-1,Y=0,Z=0)
    DriverItem=(Type=Bone,Name="upperarm_r")
    ExecutePin=()
    FalloffRegionScaleFactors=(PositiveWidth=1,NegativeWidth=1,PositiveHeight=1,NegativeHeight=1)
    FalloffSize=0.4
    FlipHeightScaling=true
    FlipWidthScaling=false
    OptionalParentItem=(Type=Bone)
    RotationOffset=(X=90,Y=0,Z=0)
   )
   > &RigUnit_ModifyTransforms(
    ItemToModify=((Item=(Type=Bone,Name="def_cloth_clavicle_r"),Transform=(Rotation=(X=0,Y=0,Z=0,W=1),Translation=(X=0,Y=0,Z=0),Scale3D=(X=1,Y=1,Z=1))))
    WeightMinimum=0
    Weight=@n22.OutputParam
   )
   n23 = &RigUnit_SphericalPoseReader(
    ActiveRegionScaleFactors=(PositiveWidth=0.5,NegativeWidth=0.5,PositiveHeight=0.5,NegativeHeight=0.5)
    ActiveRegionSize=0.1
    Debug=(bDrawDebug=true,bDraw2D=false,bDrawLocalAxes=false,DebugScale=25,DebugSegments=20,DebugThickness=0.25)
    DriverAxis=(X=-1,Y=0,Z=0)
    DriverItem=(Type=Bone,Name="upperarm_r")
    ExecutePin=()
    FalloffRegionScaleFactors=(PositiveWidth=0.2,NegativeWidth=0.2,PositiveHeight=0.2,NegativeHeight=0.2)
    FalloffSize=1
    FlipWidthScaling=False
    OptionalParentItem=(Type=Bone)
    RotationOffset=(X=25,Y=73,Z=0)
   )
   RigVMFunction_ControlFlowBranch = &RigVMFunction_ControlFlowBranch(ExecuteContext=@n23.ExecutePin, Condition=RigVMVariableNode() {variable=TacticalSprinting}.Value)
   n24 = &RigVMFunction_AlphaInterp(InterpSpeedDecreasing=5, InterpSpeedIncreasing=5, Scale=1, bClampResult=False, Value=@n23.OutputParam)
   &RigUnit_ModifyTransforms(
    ItemToModify=((Item=(Type=Bone,Name="cape_02_r"),Transform=(Rotation=(X=0,Y=-0.0436,Z=0,W=0.999),Translation=(X=0,Y=0,Z=10),Scale3D=(X=1,Y=1,Z=1))))
    Weight=@n24.Result
    ExecutePin=@RigVMFunction_ControlFlowBranch.False
   )
   n25 = &RigVMFunction_AlphaInterp(InterpSpeedDecreasing=5, InterpSpeedIncreasing=5, Scale=1, bClampResult=False, Value=@n23.OutputParam)
   > &RigUnit_ModifyTransforms(
    ItemToModify=((Item=(Type=Bone,Name="cape_mid_01_r"),Transform=(Rotation=(X=0,Y=-0.0436,Z=0,W=0.999),Translation=(X=0,Y=-2,Z=10),Scale3D=(X=1,Y=1,Z=1))))
    Weight=@n25.Result
   )
   &RigUnit_ModifyTransforms(
    ItemToModify=((Item=(Type=Bone,Name="cape_02_r"),Transform=(Rotation=(X=0,Y=-0.0436,Z=0,W=0.999),Translation=(X=0,Y=0,Z=10),Scale3D=(X=1,Y=1,Z=1))))
    ExecutePin=@RigVMFunction_ControlFlowBranch.True
    Weight=@n23.OutputParam
   )
   > &RigUnit_ModifyTransforms(
    ItemToModify=((Item=(Type=Bone,Name="cape_mid_01_r"),Transform=(Rotation=(X=0,Y=-0.0436,Z=0,W=0.999),Translation=(X=0,Y=-2,Z=10),Scale3D=(X=1,Y=1,Z=1))))
    Weight=@n23.OutputParam
   )
  ]}
]
hierarchy [
 root {loc=<0 0 0>, rot=<0 0 0 1>}
 pelvis {parent=root, loc=<0 -0.7071 93.1191>, rot=<0.0127 0.707 -0.0127 -0.707>}
 spine_01 {parent=pelvis, loc=<6.7494 -0.6388 0>, rot=<0 0 0.034 -0.9994>}
 spine_02 {parent=spine_01, loc=<7.2518 -0.035 0>, rot=<0 0 0.0238 -0.9997>}
 spine_03 {parent=spine_02, loc=<6.0104 0 0>, rot=<0 0 0.0149 -0.9999>}
 spine_04 {parent=spine_03, loc=<8.1024 0 0>, rot=<0 0 -0.0132 -0.9999>}
 spine_05 {parent=spine_04, loc=<10.0146 0 0>, rot=<0 0 -0.0015 -1>}
 attach_cape {parent=spine_05, loc=<9.0179 -11.7141 0>, rot=<0.0616 -0.7044 -0.0616 -0.7044>}
 cape_new_mid_01_MSTR {parent=attach_cape, loc=<0 2.2349 -5.9671>, rot=<-0.0111 -0.707 -0.0111 -0.707>}
 cape_new_mid_01 {parent=cape_new_mid_01_MSTR, loc=<0 0 0>, rot=<0 0 0 1>}
 cape_new_mid_02 {parent=cape_new_mid_01, loc=<12.8326 0 0>, rot=<0 0 0.0002 -1>}
 cape_new_mid_03 {parent=cape_new_mid_02, loc=<23.0451 0 0>, rot=<0 0 0 1>}
 cape_new_mid_04 {parent=cape_new_mid_03, loc=<20.3244 0 0>, rot=<0 0 0 1>}
 cape_new_mid_05 {parent=cape_new_mid_04, loc=<18.7251 0 0>, rot=<0 0 0 1>}
 cape_new_mid_06 {parent=cape_new_mid_05, loc=<14.2504 0 0>, rot=<0 0 0 1>}
 cape_new_01_mid_l_MSTR {parent=attach_cape, loc=<7.8033 2.2717 -6.0897>, rot=<0.0438 -0.6817 -0.0431 -0.729>}
 cape_new_01_mid_l {parent=cape_new_01_mid_l_MSTR, loc=<0 0 0>, rot=<0 0 0 1>}
 cape_new_02_mid_l {parent=cape_new_01_mid_l, loc=<11.8886 0 0>, rot=<0 0 -0.0247 -0.9997>}
 cape_new_03_mid_l {parent=cape_new_02_mid_l, loc=<23.3505 0 0>, rot=<0 0 0 1>}
 cape_new_04_mid_l {parent=cape_new_03_mid_l, loc=<20.0893 0 0>, rot=<0 0 0 1>}
 cape_new_05_mid_l {parent=cape_new_04_mid_l, loc=<18.5056 0 0>, rot=<0 0 0 1>}
 cape_new_06_mid_l {parent=cape_new_05_mid_l, loc=<14.6473 0 0>, rot=<0 0 0 1>}
 cape_new_01_l_MSTR {parent=attach_cape, loc=<16.7518 3.2247 -5.1697>, rot=<0.038 -0.6556 -0.0211 -0.7539>}
 cape_new_01_l {parent=cape_new_01_l_MSTR, loc=<0 0 0>, rot=<0 0 0 1>}
 cape_new_02_l {parent=cape_new_01_l, loc=<11.8886 0 0>, rot=<0 0 -0.0656 -0.9978>}
 cape_new_03_l {parent=cape_new_02_l, loc=<23.3505 0 0>, rot=<0 0 0 1>}
 cape_new_04_l {parent=cape_new_03_l, loc=<20.0893 0 0>, rot=<0 0 0 1>}
 cape_new_05_l {parent=cape_new_04_l, loc=<18.5056 0 0>, rot=<0 0 0 1>}
 cape_new_06_l {parent=cape_new_05_l, loc=<13.5867 0 0>, rot=<0 0 0 1>}
 cape_new_01_mid_r_MSTR {parent=attach_cape, loc=<-7.8033 2.2717 -6.0893>, rot=<0.729 -0.0431 0.6817 0.0438>}
 cape_new_01_mid_r {parent=cape_new_01_mid_r_MSTR, loc=<0 0 0>, rot=<0 0 0 1>}
 cape_new_02_mid_r {parent=cape_new_01_mid_r, loc=<-11.8888 -0.0001 0>, rot=<0 0 -0.0247 -0.9997>}
 cape_new_03_mid_r {parent=cape_new_02_mid_r, loc=<-23.3508 0 0>, rot=<0 0 0 1>}
 cape_new_04_mid_r {parent=cape_new_03_mid_r, loc=<-20.0893 0 0>, rot=<0 0 0 1>}
 cape_new_05_mid_r {parent=cape_new_04_mid_r, loc=<-18.5057 0 0>, rot=<0 0 0 1>}
 cape_new_06_mid_r {parent=cape_new_05_mid_r, loc=<-14.6473 0.0001 0.0001>, rot=<0 0 0 1>}
 cape_new_01_r_MSTR {parent=attach_cape, loc=<-16.7518 3.2247 -5.1694>, rot=<0.7539 -0.0211 0.6556 0.038>}
 cape_new_01_r {parent=cape_new_01_r_MSTR, loc=<0 0 0>, rot=<0 0 0 -1>}
 cape_new_02_r {parent=cape_new_01_r, loc=<-11.8886 0 -0.0001>, rot=<0 0 -0.0656 -0.9978>}
 cape_new_03_r {parent=cape_new_02_r, loc=<-23.3508 0 0.0001>, rot=<0 0 0 1>}
 cape_new_04_r {parent=cape_new_03_r, loc=<-20.0892 0 0>, rot=<0 0 0 1>}
 cape_new_05_r {parent=cape_new_04_r, loc=<-18.5057 0 0>, rot=<0 0 0 1>}
 cape_new_06_r {parent=cape_new_05_r, loc=<-13.5867 0 0>, rot=<0 0 0 1>}
 dyn_cape_MSTR {parent=attach_cape, loc=<0 2.2349 -5.9671>, rot=<-0.5078 -0.5078 0.4921 -0.4921>}
 dyn_cape_01 {parent=dyn_cape_MSTR, loc=<0 0 0>, rot=<0 0 0 1>}
 dyn_cape_02 {parent=dyn_cape_01, loc=<12.8326 0 0>, rot=<0 0.0002 0 -1>}
 dyn_cape_03 {parent=dyn_cape_02, loc=<23.0451 0 0>, rot=<0 0 0 1>}
 dyn_cape_04 {parent=dyn_cape_03, loc=<20.3244 0 0>, rot=<0 0 0 1>}
 dyn_cape_05 {parent=dyn_cape_04, loc=<18.7251 0 0>, rot=<0 0 0 1>}
 dyn_cape_06 {parent=dyn_cape_05, loc=<14.2504 0 0>, rot=<0 0 0 1>}
 dyn_cape_mid_l_MSTR {parent=attach_cape, loc=<7.8033 2.2717 -6.0897>, rot=<-0.4846 -0.5125 0.4516 -0.5465>}
 dyn_cape_mid_l_01 {parent=dyn_cape_mid_l_MSTR, loc=<0 0 0>, rot=<0 0 0 1>}
 dyn_cape_mid_l_02 {parent=dyn_cape_mid_l_01, loc=<11.8886 0 0>, rot=<0 -0.0247 0 -0.9997>}
 dyn_cape_mid_l_03 {parent=dyn_cape_mid_l_02, loc=<23.3505 0 0>, rot=<0 0 0 1>}
 dyn_cape_mid_l_04 {parent=dyn_cape_mid_l_03, loc=<20.0893 0 0>, rot=<0 0 0 1>}
 dyn_cape_mid_l_05 {parent=dyn_cape_mid_l_04, loc=<18.5056 0 0>, rot=<0 0 0 1>}
 dyn_cape_mid_l_06 {parent=dyn_cape_mid_l_05, loc=<14.6473 0 0>, rot=<0 0 0 1>}
 dyn_cape_l_MSTR {parent=attach_cape, loc=<16.7518 3.2247 -5.1697>, rot=<-0.5062 -0.4785 0.4487 -0.5599>}
 dyn_cape_l_01 {parent=dyn_cape_l_MSTR, loc=<0 0 0>, rot=<0 0 0 1>}
 dyn_cape_l_02 {parent=dyn_cape_l_01, loc=<11.8886 0 0>, rot=<0 -0.0656 0 -0.9978>}
 dyn_cape_l_03 {parent=dyn_cape_l_02, loc=<23.3505 0 0>, rot=<0 0 0 1>}
 dyn_cape_l_04 {parent=dyn_cape_l_03, loc=<20.0893 0 0>, rot=<0 0 0 1>}
 dyn_cape_l_05 {parent=dyn_cape_l_04, loc=<18.5056 0 0>, rot=<0 0 0 1>}
 dyn_cape_l_06 {parent=dyn_cape_l_05, loc=<13.5867 0 0>, rot=<0 0 0 1>}
 dyn_cape_mid_r_MSTR {parent=attach_cape, loc=<-7.8033 2.2717 -6.0893>, rot=<0.5465 0.4516 0.5125 -0.4846>}
 dyn_cape_mid_r_01 {parent=dyn_cape_mid_r_MSTR, loc=<0 0 0>, rot=<0 0 0 1>}
 dyn_cape_mid_r_02 {parent=dyn_cape_mid_r_01, loc=<-11.8888 0 0.0001>, rot=<0 -0.0247 0 -0.9997>}
 dyn_cape_mid_r_03 {parent=dyn_cape_mid_r_02, loc=<-23.3508 0 0>, rot=<0 0 0 1>}
 dyn_cape_mid_r_04 {parent=dyn_cape_mid_r_03, loc=<-20.0893 0 0>, rot=<0 0 0 1>}
 dyn_cape_mid_r_05 {parent=dyn_cape_mid_r_04, loc=<-18.5057 0 0>, rot=<0 0 0 1>}
 dyn_cape_mid_r_06 {parent=dyn_cape_mid_r_05, loc=<-14.6473 0.0001 -0.0001>, rot=<0 0 0 1>}
 dyn_cape_r_MSTR {parent=attach_cape, loc=<-16.7518 3.2247 -5.1694>, rot=<0.5599 0.4487 0.4785 -0.5062>}
 dyn_cape_r_01 {parent=dyn_cape_r_MSTR, loc=<0 0 0>, rot=<0 0 0 -1>}
 dyn_cape_r_02 {parent=dyn_cape_r_01, loc=<-11.8886 -0.0001 0>, rot=<0 -0.0656 0 -0.9978>}
 dyn_cape_r_03 {parent=dyn_cape_r_02, loc=<-23.3508 0.0001 0>, rot=<0 0 0 1>}
 dyn_cape_r_04 {parent=dyn_cape_r_03, loc=<-20.0892 0 0>, rot=<0 0 0 1>}
 dyn_cape_r_05 {parent=dyn_cape_r_04, loc=<-18.5057 0 0>, rot=<0 0 0 1>}
 dyn_cape_r_06 {parent=dyn_cape_r_05, loc=<-13.5867 0 0>, rot=<0 0 0 1>}
 neck_01 {parent=spine_05, loc=<11.3138 0.1529 0>, rot=<0 0 -0.2003 -0.9797>}
 neck_02 {parent=neck_01, loc=<4.5627 0 0>, rot=<0 0 0.0504 -0.9987>}
 head {parent=neck_02, loc=<5.3914 0 0>, rot=<0 0 0.115 -0.9934>}
 attach_backpack {parent=spine_05, loc=<-0.1219 -13.4731 0>, rot=<-0.0093 -0.707 0.0093 -0.707>}
 clavicle_l {parent=spine_05, loc=<4.3575 -0.1466 -2.3223>, rot=<0.0465 -0.7289 -0.009 -0.6829>}
 upperarm_l {parent=clavicle_l, loc=<15.5111 0 0>, rot=<0.0226 -0.3234 -0.0088 -0.9459>}
 lowerarm_l {parent=upperarm_l, loc=<26.6514 0 0>, rot=<0.0391 -0.0099 -0.2453 -0.9686>}
 hand_l {parent=lowerarm_l, loc=<22.3491 0 0>, rot=<-0.7081 -0.0507 0.0975 -0.6975>}
 thumb_01_l {parent=hand_l, loc=<1.3118 -0.3191 -2.459>, rot=<0.6546 -0.4042 0.0664 -0.6354>}
 thumb_02_l {parent=thumb_01_l, loc=<5.9336 -0.0448 0.0401>, rot=<-0.0037 0.0213 0.298 -0.9543>}
 thumb_03_l {parent=thumb_02_l, loc=<2.2223 0.3682 -0.0907>, rot=<0.0118 -0.013 0.0066 -0.9998>}
 index_metacarpal_l {parent=hand_l, loc=<1.6535 0.0183 -2.96>, rot=<0.0158 -0.0124 -0.0173 -0.9996>}
 index_01_l {parent=index_metacarpal_l, loc=<8.3008 0.0357 -0.0524>, rot=<-0.0044 0.0485 0.0924 -0.9945>}
 index_02_l {parent=index_01_l, loc=<3.3587 -0.3004 0.0452>, rot=<-0.002 0.0027 0.3225 -0.9466>}
 index_03_l {parent=index_02_l, loc=<2.5578 -0.123 -0.0252>, rot=<-0.002 0.0067 0.2558 -0.9667>}
 middle_metacarpal_l {parent=hand_l, loc=<2.1856 -0.1725 -0.5984>, rot=<0.006 0.0168 -0.0329 -0.9993>}
 middle_01_l {parent=middle_metacarpal_l, loc=<7.3275 -0.2447 0.0129>, rot=<-0.0208 0.0042 0.0858 -0.9961>}
 middle_02_l {parent=middle_01_l, loc=<4.0123 0.1401 0.1175>, rot=<-0.0218 0.0258 0.3512 -0.9357>}
 middle_03_l {parent=middle_02_l, loc=<2.6077 0.0534 0.1373>, rot=<0.0082 0.0188 0.1708 -0.9851>}
 ring_metacarpal_l {parent=hand_l, loc=<2.2074 -0.3625 1.3638>, rot=<-0.0284 0.0722 -0.0458 -0.9959>}
 ring_01_l {parent=ring_metacarpal_l, loc=<7.0295 -0.6197 -0.1446>, rot=<-0.0489 -0.0252 0.1892 -0.9804>}
 ring_02_l {parent=ring_01_l, loc=<3.6174 0.4647 -0.1027>, rot=<0 0 0.2806 -0.9598>}
 ring_03_l {parent=ring_02_l, loc=<2.4152 0.114 0.0199>, rot=<0 0 0.1189 -0.9929>}
 pinky_metacarpal_l {parent=hand_l, loc=<1.7616 -0.9753 3.3814>, rot=<-0.0397 0.0814 -0.069 -0.9935>}
 pinky_01_l {parent=pinky_metacarpal_l, loc=<7.1709 -0.4831 -0.1094>, rot=<-0.0654 -0.0035 0.2242 -0.9723>}
 pinky_02_l {parent=pinky_01_l, loc=<2.8274 0.0893 -0.1043>, rot=<0.0037 0.0011 0.2854 -0.9584>}
 pinky_03_l {parent=pinky_02_l, loc=<2.2046 0.0307 -0.0078>, rot=<-0.0002 0.0098 0.1445 -0.9895>}
 deform_wrist_dn_base_l {parent=hand_l, loc=<0 -2.2342 0>, rot=<0 0 0 -1>}
 deform_wrist_dn_l {parent=deform_wrist_dn_base_l, loc=<0 0 0>, rot=<0 0 0 1>}
 deform_wrist_up_base_l {parent=hand_l, loc=<0 2.2682 0>, rot=<0 0 0 -1>}
 deform_wrist_up_l {parent=deform_wrist_up_base_l, loc=<0 0 0>, rot=<0 0 0 1>}
 lowerarm_twist_02_l {parent=lowerarm_l, loc=<9.6199 0 0>, rot=<0 0 0 -1>}
 lowerarm_twist_01_l {parent=lowerarm_l, loc=<17.8549 0 0>, rot=<0 0 0 -1>}
 joint1 {parent=lowerarm_twist_01_l, loc=<-1.5425 5.5215 -0.9197>, rot=<0.0967 -0.0693 0.695 -0.7091>}
 joint2 {parent=lowerarm_twist_01_l, loc=<-2.174 -3.1599 -4.5741>, rot=<0.6559 0.6095 0.3388 -0.289>}
 joint3 {parent=lowerarm_twist_01_l, loc=<-1.9327 -2.2117 4.3932>, rot=<-0.1367 0.6744 0.2458 -0.6827>}
 sleeve_l_3 {parent=lowerarm_twist_01_l, loc=<-1.5425 5.5215 -0.9197>, rot=<0.0967 -0.0693 0.695 -0.7091>}
 sleeve_l_2 {parent=lowerarm_twist_01_l, loc=<-2.174 -3.1599 -4.5741>, rot=<0.6559 0.6095 0.3388 -0.289>}
 sleeve_l_1 {parent=lowerarm_twist_01_l, loc=<-1.9327 -2.2117 4.3932>, rot=<-0.1367 0.6744 0.2458 -0.6827>}
 deform_elbow_in_base_l {parent=lowerarm_l, loc=<0.2112 5.259 0.8274>, rot=<0 0 0 1>}
 deform_elbow_in_l {parent=deform_elbow_in_base_l, loc=<0 0 0>, rot=<0 0 0 1>}
 deform_elbow_out_base_l {parent=lowerarm_l, loc=<0 -4 0>, rot=<0 0 0 1>}
 deform_elbow_out_l {parent=deform_elbow_out_base_l, loc=<0 0 0>, rot=<0 0 0 1>}
 upperarm_twist_01_l {parent=upperarm_l, loc=<0.0041 0 0>, rot=<0 0 0 1>}
 deform_upperarm_bk_base_l {parent=upperarm_twist_01_l, loc=<1.0813 -7.2193 -1.9453>, rot=<0 0 0 1>}
 deform_upperarm_bk_l {parent=deform_upperarm_bk_base_l, loc=<0 0 0>, rot=<0 0 0 1>}
 deform_upperarm_fr_base_l {parent=upperarm_twist_01_l, loc=<-0.3691 7.0506 -2.765>, rot=<0 0 0 1>}
 deform_upperarm_fr_l {parent=deform_upperarm_fr_base_l, loc=<0 0 0>, rot=<0 0 0 1>}
 deform_upperarm_up_base_l {parent=upperarm_twist_01_l, loc=<-3.2216 -1.1391 3.3311>, rot=<0 0 0 1>}
 deform_upperarm_up_l {parent=deform_upperarm_up_base_l, loc=<0 0 0>, rot=<0 0 0 1>}
 deform_upperarm_up_cape_l {parent=deform_upperarm_up_l, loc=<0 0 0>, rot=<0 0 0 1>}
 upperarm_twist_02_l {parent=upperarm_l, loc=<13.7941 0 0>, rot=<0 0 0 1>}
 clavicle_cape_l {parent=clavicle_l, loc=<0 0 0>, rot=<0 0 0 1>}
 cape_01_l {parent=clavicle_cape_l, loc=<8.2 -4.7499 6.9003>, rot=<0.3413 -0.5667 0.2993 -0.6876>}
 cape_02_l {parent=cape_01_l, loc=<6.892 0 0>, rot=<0 0 -0.3403 -0.9403>}
 cape_03_l {parent=cape_02_l, loc=<11.8886 0 0>, rot=<0 0 -0.1042 -0.9946>}
 cape_04_l {parent=cape_03_l, loc=<23.3505 0 0>, rot=<0 0 0.0042 -1>}
 cape_05_l {parent=cape_04_l, loc=<20.0893 0 0>, rot=<0 0 -0.008 -1>}
 cape_06_l {parent=cape_05_l, loc=<18.5056 0 0>, rot=<0 0 -0.0041 -1>}
 cape_07_l {parent=cape_06_l, loc=<13.5867 0 0>, rot=<0 0 0.0015 -1>}
 cape_08_l {parent=cape_07_l, loc=<15.1001 0 0>, rot=<0 0 0 1>}
 clavicle_r {parent=spine_05, loc=<4.3575 -0.1466 2.3223>, rot=<0.7289 0.0465 -0.6829 0.009>}
 upperarm_r {parent=clavicle_r, loc=<-15.5111 0 0>, rot=<0.0226 -0.3234 -0.0088 -0.9459>}
 lowerarm_r {parent=upperarm_r, loc=<-26.6514 0 0>, rot=<0.0391 -0.0099 -0.2453 -0.9686>}
 hand_r {parent=lowerarm_r, loc=<-22.3491 0 0>, rot=<-0.7081 -0.0507 0.0975 -0.6975>}
 thumb_01_r {parent=hand_r, loc=<-1.3118 0.3191 2.459>, rot=<0.6546 -0.4042 0.0664 -0.6354>}
 thumb_02_r {parent=thumb_01_r, loc=<-5.9336 0.0448 -0.0401>, rot=<-0.0037 0.0213 0.298 -0.9543>}
 thumb_03_r {parent=thumb_02_r, loc=<-2.2223 -0.3682 0.0907>, rot=<0.0118 -0.013 0.0066 -0.9998>}
 index_metacarpal_r {parent=hand_r, loc=<-1.6535 -0.0183 2.96>, rot=<0.0158 -0.0124 -0.0173 -0.9996>}
 index_01_r {parent=index_metacarpal_r, loc=<-8.3008 -0.0357 0.0524>, rot=<-0.0044 0.0485 0.0924 -0.9945>}
 index_02_r {parent=index_01_r, loc=<-3.3587 0.3004 -0.0452>, rot=<-0.002 0.0027 0.3225 -0.9466>}
 index_03_r {parent=index_02_r, loc=<-2.5578 0.123 0.0252>, rot=<-0.002 0.0067 0.2558 -0.9667>}
 middle_metacarpal_r {parent=hand_r, loc=<-2.1856 0.1725 0.5984>, rot=<0.006 0.0168 -0.0329 -0.9993>}
 middle_01_r {parent=middle_metacarpal_r, loc=<-7.3275 0.2447 -0.0129>, rot=<-0.0208 0.0042 0.0858 -0.9961>}
 middle_02_r {parent=middle_01_r, loc=<-4.0123 -0.1401 -0.1175>, rot=<-0.0218 0.0258 0.3512 -0.9357>}
 middle_03_r {parent=middle_02_r, loc=<-2.6077 -0.0534 -0.1373>, rot=<0.0082 0.0188 0.1708 -0.9851>}
 ring_metacarpal_r {parent=hand_r, loc=<-2.2074 0.3625 -1.3638>, rot=<-0.0284 0.0722 -0.0458 -0.9959>}
 ring_01_r {parent=ring_metacarpal_r, loc=<-7.0295 0.6197 0.1446>, rot=<-0.0489 -0.0252 0.1892 -0.9804>}
 ring_02_r {parent=ring_01_r, loc=<-3.6174 -0.4647 0.1027>, rot=<0 0 0.2806 -0.9598>}
 ring_03_r {parent=ring_02_r, loc=<-2.4152 -0.114 -0.0199>, rot=<0 0 0.1189 -0.9929>}
 pinky_metacarpal_r {parent=hand_r, loc=<-1.7616 0.9753 -3.3814>, rot=<-0.0397 0.0814 -0.069 -0.9935>}
 pinky_01_r {parent=pinky_metacarpal_r, loc=<-7.1709 0.4831 0.1094>, rot=<-0.0654 -0.0035 0.2242 -0.9723>}
 pinky_02_r {parent=pinky_01_r, loc=<-2.8274 -0.0893 0.1043>, rot=<0.0037 0.0011 0.2854 -0.9584>}
 pinky_03_r {parent=pinky_02_r, loc=<-2.2046 -0.0307 0.0078>, rot=<-0.0002 0.0098 0.1445 -0.9895>}
 deform_wrist_dn_base_r {parent=hand_r, loc=<-0.0004 2.2345 -0.0001>, rot=<0 0 0 -1>}
 deform_wrist_dn_r {parent=deform_wrist_dn_base_r, loc=<0 0 0>, rot=<0 0 0 1>}
 deform_wrist_up_base_r {parent=hand_r, loc=<0.0003 -2.2684 0.0001>, rot=<0 0 0 -1>}
 deform_wrist_up_r {parent=deform_wrist_up_base_r, loc=<0 0 0>, rot=<0 0 0 1>}
 lowerarm_twist_02_r {parent=lowerarm_r, loc=<-9.6199 0 0>, rot=<0 0 0 -1>}
 lowerarm_twist_01_r {parent=lowerarm_r, loc=<-17.8549 0 0>, rot=<0 0 0 -1>}
 joint4 {parent=lowerarm_twist_01_r, loc=<1.5764 -5.7296 0.1816>, rot=<-0.6498 0.0321 0.7584 -0.0402>}
 joint6 {parent=lowerarm_twist_01_r, loc=<2.1673 2.6045 3.9139>, rot=<0.9703 -0.011 0.0163 0.2412>}
 joint5 {parent=lowerarm_twist_01_r, loc=<0.7245 2.3891 -5.1179>, rot=<0.6945 -0.1744 0.6943 -0.0721>}
 sleeve_r_1 {parent=lowerarm_twist_01_r, loc=<1.5764 -5.7296 0.1816>, rot=<-0.6498 0.0321 0.7584 -0.0402>}
 sleeve_r_3 {parent=lowerarm_twist_01_r, loc=<2.1673 2.6045 3.9139>, rot=<0.9703 -0.011 0.0163 0.2412>}
 sleeve_r_2 {parent=lowerarm_twist_01_r, loc=<0.7245 2.3891 -5.1179>, rot=<0.6945 -0.1744 0.6943 -0.0721>}
 deform_elbow_in_base_r {parent=lowerarm_r, loc=<-0.2111 -5.2591 -0.8275>, rot=<0 0 0 -1>}
 deform_elbow_in_r {parent=deform_elbow_in_base_r, loc=<-0.0001 0 0.0002>, rot=<0 0 0 1>}
 deform_elbow_out_base_r {parent=lowerarm_r, loc=<0 4 -0.0001>, rot=<0 0 0 -1>}
 deform_elbow_out_r {parent=deform_elbow_out_base_r, loc=<0 0 0>, rot=<0 0 0 1>}
 upperarm_twist_01_r {parent=upperarm_r, loc=<-0.0041 0 0>, rot=<0 0 0 -1>}
 deform_upperarm_bk_base_r {parent=upperarm_twist_01_r, loc=<-1.0814 7.2192 1.9455>, rot=<0 0 0 -1>}
 deform_upperarm_bk_r {parent=deform_upperarm_bk_base_r, loc=<0 0 0>, rot=<0 0 0 1>}
 deform_upperarm_fr_base_r {parent=upperarm_twist_01_r, loc=<0.3689 -7.0506 2.7652>, rot=<0 0 0 -1>}
 deform_upperarm_fr_r {parent=deform_upperarm_fr_base_r, loc=<0 0 0>, rot=<0 0 0 1>}
 deform_upperarm_up_base_r {parent=upperarm_twist_01_r, loc=<3.2214 1.1391 -3.3308>, rot=<0 0 0 -1>}
 deform_upperarm_up_r {parent=deform_upperarm_up_base_r, loc=<0 0 0>, rot=<0 0 0 -1>}
 deform_upperarm_up_cape_r {parent=deform_upperarm_up_r, loc=<0 0 0>, rot=<0 0 0 1>}
 upperarm_twist_02_r {parent=upperarm_r, loc=<-13.7941 0 0>, rot=<0 0 0 -1>}
 clavicle_cape_r {parent=clavicle_r, loc=<0 0 0>, rot=<0 0 0 1>}
 cape_01_r {parent=clavicle_cape_r, loc=<-8.2 4.7499 -6.9001>, rot=<0.5667 0.3413 0.6876 0.2993>}
 cape_02_r {parent=cape_01_r, loc=<6.8922 -0.0002 0>, rot=<0 0 -0.3403 -0.9403>}
 cape_03_r {parent=cape_02_r, loc=<11.8886 0 0>, rot=<0 0 -0.1042 -0.9946>}
 cape_04_r {parent=cape_03_r, loc=<23.3501 0.0001 0>, rot=<0 0 0.0042 -1>}
 cape_05_r {parent=cape_04_r, loc=<20.0898 0 0.0001>, rot=<0 0 -0.008 -1>}
 cape_06_r {parent=cape_05_r, loc=<18.5056 -0.0001 -0.0001>, rot=<0 0 -0.0041 -1>}
 cape_07_r {parent=cape_06_r, loc=<13.5867 0 0.0001>, rot=<0 0 0.0015 -1>}
 cape_08_r {parent=cape_07_r, loc=<15.1001 0.0001 -0.0001>, rot=<0 0 0 -1>}
 cape_root {parent=spine_05, loc=<9.0735 -9.9874 0>, rot=<-0.0087 -0.7071 0.0087 -0.7071>}
 cape_mid_01_l {parent=cape_root, loc=<5 -0.2606 -0.4394>, rot=<0.1329 -0.6858 0.1296 -0.7037>}
 cape_mid_02_l {parent=cape_mid_01_l, loc=<12.8326 0 0>, rot=<0 0 -0.0975 -0.9952>}
 cape_mid_03_l {parent=cape_mid_02_l, loc=<23.0451 0 0>, rot=<0 0 -0.0173 -0.9998>}
 cape_mid_04_l {parent=cape_mid_03_l, loc=<20.3244 0 0>, rot=<0 0 -0.0139 -0.9999>}
 cape_mid_05_l {parent=cape_mid_04_l, loc=<18.7251 0 0>, rot=<0 0 0.0005 -1>}
 cape_mid_06_l {parent=cape_mid_05_l, loc=<13.6769 0 0>, rot=<0 0 0.0016 -1>}
 cape_mid_07_l {parent=cape_mid_06_l, loc=<14.7332 0 0>, rot=<0 0 0 1>}
 cape_mid_01_r {parent=cape_root, loc=<-5 -0.2606 -0.4389>, rot=<0.1296 -0.7037 0.1329 -0.6858>}
 cape_mid_02_r {parent=cape_mid_01_r, loc=<12.8333 -0.0003 0>, rot=<0 0 -0.0975 -0.9952>}
 cape_mid_03_r {parent=cape_mid_02_r, loc=<23.0446 0 0>, rot=<0 0 -0.0173 -0.9998>}
 cape_mid_04_r {parent=cape_mid_03_r, loc=<20.3247 0 0>, rot=<0 0 -0.0139 -0.9999>}
 cape_mid_05_r {parent=cape_mid_04_r, loc=<18.7251 0 0>, rot=<0 0 0.0005 -1>}
 cape_mid_06_r {parent=cape_mid_05_r, loc=<13.6769 -0.0001 0>, rot=<0 0 0.0016 -1>}
 cape_mid_07_r {parent=cape_mid_06_r, loc=<14.7332 0 0>, rot=<0 0 0 1>}
 deform_pec_aim_base_l {parent=spine_04, loc=<7.8732 12.6388 -2.4123>, rot=<0.0102 -0.4602 0.3979 -0.7936>}
 deform_pec_aim_l {parent=deform_pec_aim_base_l, loc=<0 0 0>, rot=<0 0 0 1>}
 deform_pec_base_l {parent=deform_pec_aim_l, loc=<10.766 0 0>, rot=<0 0 0 1>}
 deform_pec_l {parent=deform_pec_base_l, loc=<0.055 0 0>, rot=<0 0 0 1>}
 deform_pec_aim_base_r {parent=spine_04, loc=<7.8727 12.6388 2.4123>, rot=<-0.4602 -0.0102 0.7936 0.3979>}
 deform_pec_aim_r {parent=deform_pec_aim_base_r, loc=<0 0 0>, rot=<0 0 0 1>}
 deform_pec_base_r {parent=deform_pec_aim_r, loc=<-10.7662 -0.0006 -0.0007>, rot=<0 0 0 1>}
 deform_pec_r {parent=deform_pec_base_r, loc=<0 0 0>, rot=<0 0 0 1>}
 thigh_l {parent=pelvis, loc=<-2.5693 -0.0109 -7.7501>, rot=<0.0012 0.0572 0.0155 -0.9982>}
 calf_l {parent=thigh_l, loc=<-39.0484 0 0>, rot=<0.0006 -0.0161 -0.0391 -0.9991>}
 foot_l {parent=calf_l, loc=<-42.561 0.1077 0.4292>, rot=<0.0418 -0.0455 0.0343 -0.9975>}
 ball_l {parent=foot_l, loc=<-6.1599 13.8264 0.6806>, rot=<0.0001 -0.0001 -0.7063 -0.7079>}
 calf_twist_02_l {parent=calf_l, loc=<-18.8064 -0.0967 0.0397>, rot=<0.0237 0.0095 -0.0035 -0.9997>}
 calf_twist_01_l {parent=calf_l, loc=<-35.813 -0.0967 0.0397>, rot=<0.0237 0.0095 -0.0035 -0.9997>}
 deform_knee_in_base_l {parent=calf_l, loc=<1.298 -5.2957 -0.0496>, rot=<0 0 0 -1>}
 deform_knee_in_l {parent=deform_knee_in_base_l, loc=<0 0 0>, rot=<0 0 0 1>}
 deform_knee_out_base_l {parent=calf_l, loc=<0 4 0>, rot=<0 0 0 -1>}
 deform_knee_out_l {parent=deform_knee_out_base_l, loc=<0 0 0>, rot=<0 0 0 1>}
 thigh_twist_01_l {parent=thigh_l, loc=<-21.6303 0 0>, rot=<0.0013 0 0 -1>}
 deform_groin_base_l {parent=thigh_l, loc=<-2.1908 8.1142 -1.7442>, rot=<0 0 0 -1>}
 deform_groin_l {parent=deform_groin_base_l, loc=<0 0 0>, rot=<0 0 0 1>}
 thigh_r {parent=pelvis, loc=<-2.5693 -0.0109 7.7501>, rot=<0.0572 -0.0012 0.9982 0.0155>}
 calf_r {parent=thigh_r, loc=<39.0484 0 0>, rot=<0.0006 -0.0161 -0.0391 -0.9991>}
 foot_r {parent=calf_r, loc=<42.561 -0.1077 -0.4292>, rot=<0.0418 -0.0455 0.0343 -0.9975>}
 ball_r {parent=foot_r, loc=<6.1599 -13.8264 -0.6806>, rot=<0.0001 -0.0001 -0.7063 -0.7079>}
 calf_twist_02_r {parent=calf_r, loc=<18.8064 0.0967 -0.0397>, rot=<0.0237 0.0095 -0.0035 -0.9997>}
 calf_twist_01_r {parent=calf_r, loc=<35.813 0.0967 -0.0397>, rot=<0.0237 0.0095 -0.0035 -0.9997>}
 deform_knee_in_base_r {parent=calf_r, loc=<-1.298 5.2957 0.0496>, rot=<0 0 0 -1>}
 deform_knee_in_r {parent=deform_knee_in_base_r, loc=<0 0 0>, rot=<0 0 0 1>}
 deform_knee_out_base_r {parent=calf_r, loc=<0 -4 0>, rot=<0 0 0 -1>}
 deform_knee_out_r {parent=deform_knee_out_base_r, loc=<0 0 0>, rot=<0 0 0 1>}
 calf_r_extra {parent=calf_r, loc=<0 0 0>, rot=<0 0 0 1>}
 thigh_twist_01_r {parent=thigh_r, loc=<21.6303 0 0>, rot=<0.0013 0 0 -1>}
 deform_groin_base_r {parent=thigh_r, loc=<2.1908 -8.1142 1.7442>, rot=<0 0 0 -1>}
 deform_groin_r {parent=deform_groin_base_r, loc=<0 0 0>, rot=<0 0 0 1>}
 deform_glute_base_l {parent=pelvis, loc=<-8.8064 -8.8222 -8.4287>, rot=<0.0275 0.0501 -0.4472 -0.8926>}
 deform_glute_l {parent=deform_glute_base_l, loc=<0 0 0>, rot=<0 0 0 1>}
 deform_glute_base_r {parent=pelvis, loc=<-8.8062 -8.8222 8.4287>, rot=<0.0501 -0.0275 0.8926 -0.4472>}
 deform_glute_r {parent=deform_glute_base_r, loc=<0 0 0>, rot=<0 0 0 1>}
 dyn_ground_collision {parent=root, loc=<0 0 0>, rot=<0 0 0 1>}
]
variables [
 IsSkydiving {type=bool, default=False}
 TacticalSprinting {type=bool, default=False}
 backpack_wearing {type=bool, default=False}
 parachute {type=double, default=0}
 tacticalSprint_and_jump {type=double, default=0}
]
