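# Physics asset for the SesameSeed medium male player body, simulating its cape, with a dedicated back bling collision capsule.
# Five six-segment capsule chains hang from master bodies locked to spine_05, with angle-limited joints along each.
# Free-angle cross-links join chains from segment 02 down; on the left, center segments 04–06 link to the outer chain, skipping mid-left.
# Capsules cover the body and limbs; a huge root sphere sits below, apparently as ground.
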PhysicsAsset /Game/Characters/Player/Male/Medium/Bodies/M_MED_SesameSeed/Meshes/M_MED_SesameSeed_Physics_Cape_BackblingRework
bodies {
 calf_l [{shapeType=Capsule, length=26.6823, radius=11.3606, center=<-27.1113 2.2029 1.1091>, rotation=<rot 87.9768 -4.7743 0.0001>}]
 calf_r [{shapeType=Capsule, length=29.1078, radius=11.5553, center=<28.2139 -3.2456 -1.4583>, rotation=<rot 88.6662 -4.809 0.0001>}]
 clavicle_l [{shapeType=Capsule, length=9.1218, radius=9.1836, center=<5.8371 -0.6839 -1.5328>, rotation=<rot -77.5566 177.3915 7.7314>}]
 clavicle_r [{shapeType=Capsule, length=6.768, radius=9.2241, center=<-4.0507 1.2527 1.754>, rotation=<rot -77.5565 177.3915 -172.2685>}]
 dyn_backbling_collision [{
   shapeType=Capsule
   length=22.125
   radius=14
   shapeName=dyn_backbling_collision_capsule
   center=<0 -24.8614 -4.1972>
   rotation=<rot 9.7827 -90 -90>
  }]
 dyn_cape_01 [{shapeType=Capsule, length=12.375, radius=2.625, shapeName=dyn_cape_01_capsule, center=<4 0 2.5>, rotation=<rot 90 0 0>}]
 dyn_cape_02 [{shapeType=Capsule, length=9.8125, radius=5.1875, shapeName=dyn_cape_02_capsule, center=<5.4248 0 2.5>, rotation=<rot 90 0 0>}]
 dyn_cape_03 [{shapeType=Capsule, length=8.875, radius=6.125, shapeName=dyn_cape_03_capsule, center=<4 0 2.5>, rotation=<rot 90 0 0>}]
 dyn_cape_04 [{shapeType=Capsule, length=8.6875, radius=6.3125, shapeName=dyn_cape_04_capsule, center=<4 0 2.5>, rotation=<rot 90 0 0>}]
 dyn_cape_05 [{shapeType=Capsule, length=8, radius=7, shapeName=dyn_cape_05_capsule, center=<4 0 2.5>, rotation=<rot 90 0 0>}]
 dyn_cape_06 [{shapeType=Capsule, length=4.5, radius=10.5, shapeName=dyn_cape_06_capsule, center=<4 0 2.5>, rotation=<rot 90 0 0>}]
 dyn_cape_MSTR [{shapeType=Capsule, length=0.505, radius=0.505, shapeName=dyn_cape_MSTR_capsule, center=<0 0 2.5>, rotation=<rot 0 -90 0>}]
 dyn_cape_l_01 [{shapeType=Capsule, length=11.1875, radius=3.8125, shapeName=dyn_cape_l_01_capsule, center=<4 0 2.5>, rotation=<rot 90 0 0>}]
 dyn_cape_l_02 [{shapeType=Capsule, length=11.5312, radius=5.9688, shapeName=dyn_cape_l_02_capsule, center=<7.5664 0 2.5>, rotation=<rot 90 0 0>}]
 dyn_cape_l_03 [{shapeType=Capsule, length=8.875, radius=6.125, shapeName=dyn_cape_l_03_capsule, center=<4 0 2.5>, rotation=<rot 90 0 0>}]
 dyn_cape_l_04 [{shapeType=Capsule, length=8.6875, radius=6.3125, shapeName=dyn_cape_l_04_capsule, center=<4 0 2.5>, rotation=<rot 90 0 0>}]
 dyn_cape_l_05 [{shapeType=Capsule, length=8, radius=7, shapeName=dyn_cape_l_05_capsule, center=<4 0 2.5>, rotation=<rot 90 0 0>}]
 dyn_cape_l_06 [{shapeType=Capsule, length=4.5, radius=10.5, shapeName=dyn_cape_l_06_capsule, center=<4 0 2.5>, rotation=<rot 90 0 0>}]
 dyn_cape_l_MSTR [{shapeType=Capsule, length=0.505, radius=0.505, shapeName=dyn_cape_l_MSTR_capsule, center=<0 0 2.5>, rotation=<rot 0 -90 0>}]
 dyn_cape_mid_l_01 [{shapeType=Capsule, length=12.375, radius=2.625, shapeName=dyn_cape_mid_l_01_capsule, center=<4 0 2.5>, rotation=<rot 90 0 0>}]
 dyn_cape_mid_l_02 [{shapeType=Capsule, length=9.75, radius=5.25, shapeName=dyn_cape_mid_l_02_capsule, center=<6.144 0 2.5>, rotation=<rot 90 0 0>}]
 dyn_cape_mid_l_03 [{shapeType=Capsule, length=8.875, radius=6.125, shapeName=dyn_cape_mid_l_03_capsule, center=<4 0 2.5>, rotation=<rot 90 0 0>}]
 dyn_cape_mid_l_04 [{shapeType=Capsule, length=8.6875, radius=6.3125, shapeName=dyn_cape_mid_l_04_capsule, center=<4 0 2.5>, rotation=<rot 90 0 0>}]
 dyn_cape_mid_l_05 [{shapeType=Capsule, length=8, radius=7, shapeName=dyn_cape_mid_l_05_capsule, center=<4 0 2.5>, rotation=<rot 90 0 0>}]
 dyn_cape_mid_l_06 [{shapeType=Capsule, length=4.5, radius=10.5, shapeName=dyn_cape_mid_l_06_capsule, center=<4 0 2.5>, rotation=<rot 90 0 0>}]
 dyn_cape_mid_l_MSTR [{shapeType=Capsule, length=0.505, radius=0.505, shapeName=dyn_cape_mid_l_MSTR_capsule, center=<0 0 2.5>, rotation=<rot 0 -90 0>}]
 dyn_cape_mid_r_01 [{shapeType=Capsule, length=12.375, radius=2.625, shapeName=dyn_cape_mid_r_01_capsule, center=<-4 0 -2.5>, rotation=<rot 90 0 0>}]
 dyn_cape_mid_r_02 [{shapeType=Capsule, length=9.75, radius=5.25, shapeName=dyn_cape_mid_r_02_capsule, center=<-5.8798 0 -2.5>, rotation=<rot 90 0 0>}]
 dyn_cape_mid_r_03 [{shapeType=Capsule, length=8.875, radius=6.125, shapeName=dyn_cape_mid_r_03_capsule, center=<-4 0 -2.5>, rotation=<rot 90 0 0>}]
 dyn_cape_mid_r_04 [{shapeType=Capsule, length=8.6875, radius=6.3125, shapeName=dyn_cape_mid_r_04_capsule, center=<-4 0 -2.5>, rotation=<rot 90 0 0>}]
 dyn_cape_mid_r_05 [{shapeType=Capsule, length=8, radius=7, shapeName=dyn_cape_mid_r_05_capsule, center=<-4 0 -2.5>, rotation=<rot 90 0 0>}]
 dyn_cape_mid_r_06 [{shapeType=Capsule, length=4.5, radius=10.5, shapeName=dyn_cape_mid_r_06_capsule, center=<-4 0 -2.5>, rotation=<rot 90 0 0>}]
 dyn_cape_mid_r_MSTR [{shapeType=Capsule, length=0.505, radius=0.505, shapeName=dyn_cape_mid_r_MSTR_capsule, center=<0 0 -2.5>, rotation=<rot 0 -90 0>}]
 dyn_cape_r_01 [{shapeType=Capsule, length=11.375, radius=3.625, shapeName=dyn_cape_r_01_capsule, center=<-4 0 -2.5>, rotation=<rot 90 0 0>}]
 dyn_cape_r_02 [{shapeType=Capsule, length=11.5312, radius=5.9688, shapeName=dyn_cape_r_02_capsule, center=<-7.0317 -1.2211 -2.5>, rotation=<rot 90 0 0>}]
 dyn_cape_r_03 [{shapeType=Capsule, length=8.875, radius=6.125, shapeName=dyn_cape_r_03_capsule, center=<-4 0 -2.5>, rotation=<rot 90 0 0>}]
 dyn_cape_r_04 [{shapeType=Capsule, length=8.6875, radius=6.3125, shapeName=dyn_cape_r_04_capsule, center=<-4 0 -2.5>, rotation=<rot 90 0 0>}]
 dyn_cape_r_05 [{shapeType=Capsule, length=8, radius=7, shapeName=dyn_cape_r_05_capsule, center=<-4 0 -2.5>, rotation=<rot 90 0 0>}]
 dyn_cape_r_06 [{shapeType=Capsule, length=4.5, radius=10.5, shapeName=dyn_cape_r_06_capsule, center=<-4 0 -2.5>, rotation=<rot 90 0 0>}]
 dyn_cape_r_MSTR [{shapeType=Capsule, length=0.505, radius=0.505, shapeName=dyn_cape_r_MSTR_capsule, center=<0 0 -2.5>, rotation=<rot 0 -90 0>}]
 foot_l [{
   shapeType=Capsule
   length=3.8141
   radius=15.1267
   shapeName=foot_l_capsule
   center=<-5.4831 4.6158 -0.0809>
   rotation=<rot -0.3028 8.5244 -87.841>
  }]
 foot_r [{
   shapeType=Capsule
   length=3.8141
   radius=15.1267
   shapeName=foot_l_capsule
   center=<5.4831 -4.6158 0.0809>
   rotation=<rot -0.3028 8.5244 92.159>
  }]
 hand_l [{shapeType=Capsule, length=9.0025, radius=6.3332, center=<7.7161 -2.7197 0.0627>, rotation=<rot 0 84.0522 -81.1617>}]
 hand_r [{shapeType=Capsule, length=8.9646, radius=6.3467, center=<-7.7475 2.7241 -0.0749>, rotation=<rot 0 84.0978 -83.219>}]
 lowerarm_l [{
   shapeType=Capsule
   length=19.875
   radius=6.0312
   shapeName=lowerarm_l_capsule
   center=<9.1567 0.0011 -0.6856>
   rotation=<rot -85.7181 179.9271 -179.9207>
  }]
 lowerarm_r [{
   shapeType=Capsule
   length=19.7812
   radius=6.125
   shapeName=lowerarm_l_capsule
   center=<-9.1567 -0.0011 0.6856>
   rotation=<rot -85.7181 179.9271 0.0793>
  }]
 pelvis [{shapeType=Capsule, length=8.9028, radius=11.1596, center=<-1.319 -0.94 -0.1981>}]
 root [{shapeType=Sphere, radius=242.875, shapeName=root_sphere, center=<0 -8.6403 -240.365>}]
 spine_01 [{shapeType=Capsule, length=13.2671, radius=10.4854, center=<-0.6411 0.0956 0>}]
 spine_02 [{shapeType=Capsule, length=12.5, radius=10.9854, center=<1.179 2.2974 0>}]
 spine_03 [{shapeType=Capsule, length=13.375, radius=10.1104, center=<1.2429 0.8324 0>}]
 spine_04 [{shapeType=Capsule, length=16.0796, radius=10.3916, center=<3.212 0.4573 0>}]
 spine_05 [{shapeType=Capsule, length=17.7359, radius=9.3291, center=<1.5198 -0.8941 0>}]
 thigh_l [{shapeType=Capsule, length=26.4884, radius=9.2616, shapeName=thigh_l_capsule, center=<-21.2554 -0.3501 0>, rotation=<rot 90 89.6476 89.2108>}]
 thigh_r [{shapeType=Capsule, length=30.9786, radius=9.5214, shapeName=thigh_r_capsule, center=<21.2457 0 0>, rotation=<rot 90 0 0>}]
 upperarm_l [{
   shapeType=Capsule
   length=25.75
   radius=6.3125
   shapeName=upperarm_l_capsule
   center=<9.2439 0.0001 0.2922>
   rotation=<rot -88.8289 179.9808 -179.9804>
  }]
 upperarm_r [{
   shapeType=Capsule
   length=24.3438
   radius=6
   shapeName=upperarm_l_capsule
   center=<-9.24 1.0452 0.5922>
   rotation=<rot -88.8289 179.9756 3.4902>
  }]
}
constraints [
 {
  bone1Name=pelvis
  bone2Name=root
  swing1LimitDegrees=45
  swing1Motion=Free
  swing2LimitDegrees=45
  swing2Motion=Free
  twistLimitDegrees=45
  twistMotion=Free
 }
 {
  bone1Name=spine_01
  bone2Name=pelvis
  swing1LimitDegrees=45
  swing1Motion=Free
  swing2LimitDegrees=45
  swing2Motion=Free
  twistLimitDegrees=45
  twistMotion=Free
 }
 {
  bone1Name=spine_02
  bone2Name=spine_01
  swing1LimitDegrees=45
  swing1Motion=Free
  swing2LimitDegrees=45
  swing2Motion=Free
  twistLimitDegrees=45
  twistMotion=Free
 }
 {bone1Name=spine_03, bone2Name=spine_02, swing1LimitDegrees=45, swing2LimitDegrees=45, twistLimitDegrees=45}
 {bone1Name=spine_04, bone2Name=spine_03, swing1LimitDegrees=45, swing2LimitDegrees=45, twistLimitDegrees=45}
 {bone1Name=spine_05, bone2Name=spine_04, swing1LimitDegrees=45, swing2LimitDegrees=45, twistLimitDegrees=45}
 {
  bone1Name=dyn_cape_l_MSTR
  bone2Name=spine_05
  swing1LimitDegrees=45
  swing1Motion=Locked
  swing2LimitDegrees=45
  swing2Motion=Locked
  twistLimitDegrees=45
  twistMotion=Locked
 }
 {bone1Name=dyn_cape_l_01, bone2Name=dyn_cape_l_MSTR, swing1LimitDegrees=5, swing2LimitDegrees=5, twistLimitDegrees=6}
 {bone1Name=dyn_cape_l_02, bone2Name=dyn_cape_l_01, swing1LimitDegrees=5, swing2LimitDegrees=5, twistLimitDegrees=12}
 {
  bone1Name=dyn_cape_mid_l_02
  bone2Name=dyn_cape_l_02
  swing1LimitDegrees=45
  swing1Motion=Free
  swing2LimitDegrees=45
  swing2Motion=Free
  twistLimitDegrees=45
  twistMotion=Free
 }
 {bone1Name=dyn_cape_l_03, bone2Name=dyn_cape_l_02, swing1LimitDegrees=12, swing2LimitDegrees=45, twistLimitDegrees=12}
 {
  bone1Name=dyn_cape_mid_l_03
  bone2Name=dyn_cape_l_03
  swing1LimitDegrees=45
  swing1Motion=Free
  swing2LimitDegrees=45
  swing2Motion=Free
  twistLimitDegrees=45
  twistMotion=Free
 }
 {bone1Name=dyn_cape_l_04, bone2Name=dyn_cape_l_03, swing1LimitDegrees=24, swing2LimitDegrees=45, twistLimitDegrees=12}
 {
  bone1Name=dyn_cape_04
  bone2Name=dyn_cape_l_04
  swing1LimitDegrees=45
  swing1Motion=Free
  swing2LimitDegrees=45
  swing2Motion=Free
  twistLimitDegrees=45
  twistMotion=Free
 }
 {
  bone1Name=dyn_cape_mid_l_04
  bone2Name=dyn_cape_l_04
  swing1LimitDegrees=45
  swing1Motion=Free
  swing2LimitDegrees=45
  swing2Motion=Free
  twistLimitDegrees=45
  twistMotion=Free
 }
 {bone1Name=dyn_cape_l_05, bone2Name=dyn_cape_l_04, swing1LimitDegrees=24, swing2LimitDegrees=45, twistLimitDegrees=12}
 {
  bone1Name=dyn_cape_05
  bone2Name=dyn_cape_l_05
  swing1LimitDegrees=45
  swing1Motion=Free
  swing2LimitDegrees=45
  swing2Motion=Free
  twistLimitDegrees=45
  twistMotion=Free
 }
 {
  bone1Name=dyn_cape_mid_l_05
  bone2Name=dyn_cape_l_05
  swing1LimitDegrees=45
  swing1Motion=Free
  swing2LimitDegrees=45
  swing2Motion=Free
  twistLimitDegrees=45
  twistMotion=Free
 }
 {bone1Name=dyn_cape_l_06, bone2Name=dyn_cape_l_05, swing1LimitDegrees=4, swing2LimitDegrees=4, twistLimitDegrees=4}
 {
  bone1Name=dyn_cape_06
  bone2Name=dyn_cape_l_06
  swing1LimitDegrees=45
  swing1Motion=Free
  swing2LimitDegrees=45
  swing2Motion=Free
  twistLimitDegrees=45
  twistMotion=Free
 }
 {
  bone1Name=dyn_cape_mid_l_06
  bone2Name=dyn_cape_l_06
  swing1LimitDegrees=45
  swing1Motion=Free
  swing2LimitDegrees=45
  swing2Motion=Free
  twistLimitDegrees=45
  twistMotion=Free
 }
 {
  bone1Name=dyn_cape_mid_l_MSTR
  bone2Name=spine_05
  swing1LimitDegrees=45
  swing1Motion=Locked
  swing2LimitDegrees=45
  swing2Motion=Locked
  twistLimitDegrees=45
  twistMotion=Locked
 }
 {bone1Name=dyn_cape_mid_l_01, bone2Name=dyn_cape_mid_l_MSTR, swing1LimitDegrees=5, swing2LimitDegrees=5, twistLimitDegrees=6}
 {bone1Name=dyn_cape_mid_l_02, bone2Name=dyn_cape_mid_l_01, swing1LimitDegrees=5, swing2LimitDegrees=5, twistLimitDegrees=12}
 {
  bone1Name=dyn_cape_02
  bone2Name=dyn_cape_mid_l_02
  swing1LimitDegrees=45
  swing1Motion=Free
  swing2LimitDegrees=45
  swing2Motion=Free
  twistLimitDegrees=45
  twistMotion=Free
 }
 {bone1Name=dyn_cape_mid_l_03, bone2Name=dyn_cape_mid_l_02, swing1LimitDegrees=12, swing2LimitDegrees=45, twistLimitDegrees=12}
 {
  bone1Name=dyn_cape_03
  bone2Name=dyn_cape_mid_l_03
  swing1LimitDegrees=45
  swing1Motion=Free
  swing2LimitDegrees=45
  swing2Motion=Free
  twistLimitDegrees=45
  twistMotion=Free
 }
 {bone1Name=dyn_cape_mid_l_04, bone2Name=dyn_cape_mid_l_03, swing1LimitDegrees=24, swing2LimitDegrees=45, twistLimitDegrees=12}
 {bone1Name=dyn_cape_mid_l_05, bone2Name=dyn_cape_mid_l_04, swing1LimitDegrees=24, swing2LimitDegrees=45, twistLimitDegrees=12}
 {bone1Name=dyn_cape_mid_l_06, bone2Name=dyn_cape_mid_l_05, swing1LimitDegrees=4, swing2LimitDegrees=4, twistLimitDegrees=4}
 {
  bone1Name=dyn_cape_mid_r_MSTR
  bone2Name=spine_05
  swing1LimitDegrees=45
  swing1Motion=Locked
  swing2LimitDegrees=45
  swing2Motion=Locked
  twistLimitDegrees=45
  twistMotion=Locked
 }
 {bone1Name=dyn_cape_mid_r_01, bone2Name=dyn_cape_mid_r_MSTR, swing1LimitDegrees=5, swing2LimitDegrees=5, twistLimitDegrees=6}
 {bone1Name=dyn_cape_mid_r_02, bone2Name=dyn_cape_mid_r_01, swing1LimitDegrees=5, swing2LimitDegrees=5, twistLimitDegrees=12}
 {
  bone1Name=dyn_cape_02
  bone2Name=dyn_cape_mid_r_02
  swing1LimitDegrees=45
  swing1Motion=Free
  swing2LimitDegrees=45
  swing2Motion=Free
  twistLimitDegrees=45
  twistMotion=Free
 }
 {
  bone1Name=dyn_cape_mid_r_02
  bone2Name=dyn_cape_r_02
  swing1LimitDegrees=45
  swing1Motion=Free
  swing2LimitDegrees=45
  swing2Motion=Free
  twistLimitDegrees=45
  twistMotion=Free
 }
 {bone1Name=dyn_cape_mid_r_03, bone2Name=dyn_cape_mid_r_02, swing1LimitDegrees=12, swing2LimitDegrees=45, twistLimitDegrees=12}
 {
  bone1Name=dyn_cape_03
  bone2Name=dyn_cape_mid_r_03
  swing1LimitDegrees=45
  swing1Motion=Free
  swing2LimitDegrees=45
  swing2Motion=Free
  twistLimitDegrees=45
  twistMotion=Free
 }
 {
  bone1Name=dyn_cape_mid_r_03
  bone2Name=dyn_cape_r_03
  swing1LimitDegrees=45
  swing1Motion=Free
  swing2LimitDegrees=45
  swing2Motion=Free
  twistLimitDegrees=45
  twistMotion=Free
 }
 {bone1Name=dyn_cape_mid_r_04, bone2Name=dyn_cape_mid_r_03, swing1LimitDegrees=24, swing2LimitDegrees=45, twistLimitDegrees=12}
 {
  bone1Name=dyn_cape_04
  bone2Name=dyn_cape_mid_r_04
  swing1LimitDegrees=45
  swing1Motion=Free
  swing2LimitDegrees=45
  swing2Motion=Free
  twistLimitDegrees=45
  twistMotion=Free
 }
 {
  bone1Name=dyn_cape_mid_r_04
  bone2Name=dyn_cape_r_04
  swing1LimitDegrees=45
  swing1Motion=Free
  swing2LimitDegrees=45
  swing2Motion=Free
  twistLimitDegrees=45
  twistMotion=Free
 }
 {bone1Name=dyn_cape_mid_r_05, bone2Name=dyn_cape_mid_r_04, swing1LimitDegrees=24, swing2LimitDegrees=45, twistLimitDegrees=12}
 {
  bone1Name=dyn_cape_05
  bone2Name=dyn_cape_mid_r_05
  swing1LimitDegrees=45
  swing1Motion=Free
  swing2LimitDegrees=45
  swing2Motion=Free
  twistLimitDegrees=45
  twistMotion=Free
 }
 {
  bone1Name=dyn_cape_mid_r_05
  bone2Name=dyn_cape_r_05
  swing1LimitDegrees=45
  swing1Motion=Free
  swing2LimitDegrees=45
  swing2Motion=Free
  twistLimitDegrees=45
  twistMotion=Free
 }
 {bone1Name=dyn_cape_mid_r_06, bone2Name=dyn_cape_mid_r_05, swing1LimitDegrees=4, swing2LimitDegrees=4, twistLimitDegrees=4}
 {
  bone1Name=dyn_cape_06
  bone2Name=dyn_cape_mid_r_06
  swing1LimitDegrees=45
  swing1Motion=Free
  swing2LimitDegrees=45
  swing2Motion=Free
  twistLimitDegrees=45
  twistMotion=Free
 }
 {
  bone1Name=dyn_cape_mid_r_06
  bone2Name=dyn_cape_r_06
  swing1LimitDegrees=45
  swing1Motion=Free
  swing2LimitDegrees=45
  swing2Motion=Free
  twistLimitDegrees=45
  twistMotion=Free
 }
 {
  bone1Name=dyn_cape_r_MSTR
  bone2Name=spine_05
  swing1LimitDegrees=45
  swing1Motion=Locked
  swing2LimitDegrees=45
  swing2Motion=Locked
  twistLimitDegrees=45
  twistMotion=Locked
 }
 {bone1Name=dyn_cape_r_01, bone2Name=dyn_cape_r_MSTR, swing1LimitDegrees=5, swing2LimitDegrees=5, twistLimitDegrees=6}
 {bone1Name=dyn_cape_r_02, bone2Name=dyn_cape_r_01, swing1LimitDegrees=5, swing2LimitDegrees=5, twistLimitDegrees=12}
 {bone1Name=dyn_cape_r_03, bone2Name=dyn_cape_r_02, swing1LimitDegrees=12, swing2LimitDegrees=45, twistLimitDegrees=12}
 {bone1Name=dyn_cape_r_04, bone2Name=dyn_cape_r_03, swing1LimitDegrees=24, swing2LimitDegrees=45, twistLimitDegrees=12}
 {bone1Name=dyn_cape_r_05, bone2Name=dyn_cape_r_04, swing1LimitDegrees=24, swing2LimitDegrees=45, twistLimitDegrees=12}
 {bone1Name=dyn_cape_r_06, bone2Name=dyn_cape_r_05, swing1LimitDegrees=4, swing2LimitDegrees=4, twistLimitDegrees=4}
 {
  bone1Name=dyn_cape_MSTR
  bone2Name=spine_05
  swing1LimitDegrees=45
  swing1Motion=Locked
  swing2LimitDegrees=45
  swing2Motion=Locked
  twistLimitDegrees=45
  twistMotion=Locked
 }
 {bone1Name=dyn_cape_01, bone2Name=dyn_cape_MSTR, swing1LimitDegrees=5, swing2LimitDegrees=5, twistLimitDegrees=6}
 {bone1Name=dyn_cape_02, bone2Name=dyn_cape_01, swing1LimitDegrees=5, swing2LimitDegrees=5, twistLimitDegrees=12}
 {bone1Name=dyn_cape_03, bone2Name=dyn_cape_02, swing1LimitDegrees=12, swing2LimitDegrees=45, twistLimitDegrees=12}
 {bone1Name=dyn_cape_04, bone2Name=dyn_cape_03, swing1LimitDegrees=24, swing2LimitDegrees=45, twistLimitDegrees=12}
 {bone1Name=dyn_cape_05, bone2Name=dyn_cape_04, swing1LimitDegrees=24, swing2LimitDegrees=45, twistLimitDegrees=12}
 {bone1Name=dyn_cape_06, bone2Name=dyn_cape_05, swing1LimitDegrees=4, swing2LimitDegrees=4, twistLimitDegrees=4}
 {bone1Name=clavicle_l, bone2Name=spine_05, swing1LimitDegrees=45, swing2LimitDegrees=45, twistLimitDegrees=45}
 {
  bone1Name=upperarm_l
  bone2Name=clavicle_l
  swing1LimitDegrees=45
  swing1Motion=Free
  swing2LimitDegrees=45
  swing2Motion=Free
  twistLimitDegrees=45
  twistMotion=Free
 }
 {
  bone1Name=lowerarm_l
  bone2Name=upperarm_l
  swing1LimitDegrees=45
  swing1Motion=Free
  swing2LimitDegrees=45
  swing2Motion=Free
  twistLimitDegrees=45
  twistMotion=Free
 }
 {bone1Name=hand_l, bone2Name=lowerarm_l, swing1LimitDegrees=45, swing2LimitDegrees=45, twistLimitDegrees=45}
 {bone1Name=clavicle_r, bone2Name=spine_05, swing1LimitDegrees=45, swing2LimitDegrees=45, twistLimitDegrees=45}
 {
  bone1Name=upperarm_r
  bone2Name=clavicle_r
  swing1LimitDegrees=45
  swing1Motion=Free
  swing2LimitDegrees=45
  swing2Motion=Free
  twistLimitDegrees=45
  twistMotion=Free
 }
 {
  bone1Name=lowerarm_r
  bone2Name=upperarm_r
  swing1LimitDegrees=45
  swing1Motion=Free
  swing2LimitDegrees=45
  swing2Motion=Free
  twistLimitDegrees=45
  twistMotion=Free
 }
 {bone1Name=hand_r, bone2Name=lowerarm_r, swing1LimitDegrees=45, swing2LimitDegrees=45, twistLimitDegrees=45}
 {bone1Name=thigh_l, bone2Name=pelvis, swing1LimitDegrees=45, swing2LimitDegrees=45, twistLimitDegrees=45}
 {bone1Name=calf_l, bone2Name=thigh_l, swing1LimitDegrees=45, swing2LimitDegrees=45, twistLimitDegrees=45}
 {
  bone1Name=foot_l
  bone2Name=calf_l
  swing1LimitDegrees=45
  swing1Motion=Free
  swing2LimitDegrees=45
  swing2Motion=Free
  twistLimitDegrees=45
  twistMotion=Free
 }
 {bone1Name=thigh_r, bone2Name=pelvis, swing1LimitDegrees=45, swing2LimitDegrees=45, twistLimitDegrees=45}
 {bone1Name=calf_r, bone2Name=thigh_r, swing1LimitDegrees=45, swing2LimitDegrees=45, twistLimitDegrees=45}
 {
  bone1Name=foot_r
  bone2Name=calf_r
  swing1LimitDegrees=45
  swing1Motion=Free
  swing2LimitDegrees=45
  swing2Motion=Free
  twistLimitDegrees=45
  twistMotion=Free
 }
 {
  bone1Name=dyn_backbling_collision
  bone2Name=spine_05
  swing1LimitDegrees=45
  swing1Motion=Free
  swing2LimitDegrees=45
  swing2Motion=Free
  twistLimitDegrees=45
  twistMotion=Free
 }
]
